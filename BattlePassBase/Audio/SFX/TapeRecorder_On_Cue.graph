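# Sound cue that plays the tape recorder switch-on effect for the Battle Pass.
# A single wave player feeds the TapeRecorder_On wave straight to the output, with no randomization or modulation.
# It routes through the Fort_VO_Narrator sound class rather than an SFX class and sets a very large max distance.
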SoundCue /BattlePassBase/Audio/SFX/TapeRecorder_On_Cue
root=WavePlayer_0
nodes [{id=WavePlayer_0, properties {soundWaveAssetPtr=&/TapeRecorder_On}}]
properties {maxDistance=2097152, soundClassObject=/Game/Sounds/FortSoundClassesAndModes/Fort_VO_Narrator, volumeMultiplier=1}
